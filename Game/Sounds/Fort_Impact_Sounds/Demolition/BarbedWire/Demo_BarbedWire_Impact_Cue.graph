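# Sound cue for barbed wire demolition impacts, routed to the impacts sound class with small-object hit attenuation and limited concurrency.
# A quality-level node picks among four branches with slight random pitch variation.
# The first two layer a random barbed wire hit with a random, sharply pitched-up fence impact, the second using fewer variations.
# The last two each play a single optimized wave.
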SoundCue /Game/Sounds/Fort_Impact_Sounds/Demolition/BarbedWire/Demo_BarbedWire_Impact_Cue
root=QualityLevel_0
nodes [
 {id=Mixer_0, children [Random_0,Modulator_0], properties {inputVolume [1,1]}}
 {id=Mixer_1, children [Random_2,Modulator_3], properties {inputVolume [1,1]}}
 {id=Modulator_0, children [Random_8], properties {pitchMax=1.5, pitchMin=1.4}}
 {id=Modulator_1, children [Mixer_1], properties {pitchMax=1.07, pitchMin=0.92}}
 {id=Modulator_2, children [Mixer_0], properties {pitchMax=1.07, pitchMin=0.92}}
 {id=Modulator_3, children [Random_1], properties {pitchMax=1.5, pitchMin=1.4}}
 {id=Modulator_4, children [WavePlayer_9], properties {pitchMax=1.1, pitchMin=0.9}}
 {id=Modulator_5, children [WavePlayer_10], properties {pitchMax=1.1, pitchMin=0.9}}
 {id=QualityLevel_0, children [Modulator_2,Modulator_1,Modulator_4,Modulator_5]}
 {id=Random_0, children [WavePlayer_0,WavePlayer_1,WavePlayer_2], properties {weights [1,1,1]}}
 {id=Random_1, children [WavePlayer_5,WavePlayer_3], properties {weights [1,1]}}
 {id=Random_2, children [WavePlayer_6,WavePlayer_7], properties {weights [1,1]}}
 {id=Random_8, children [WavePlayer_33, WavePlayer_34, WavePlayer_35], properties {weights [1,1,1]}}
 {id=WavePlayer_0, properties {soundWaveAssetPtr=&/barbedwire_impact_01}}
 {id=WavePlayer_1, properties {soundWaveAssetPtr=&/barbedwire_impact_02}}
 {id=WavePlayer_10, properties {soundWaveAssetPtr=&/BarbedWire_Impact_opt}}
 {id=WavePlayer_2, properties {soundWaveAssetPtr=&/barbedwire_impact_03}}
 {id=WavePlayer_3, properties {soundWaveAssetPtr=/Game/Sounds/Fort_Impact_Sounds/Demolition/Fence/Fence_Impact_Med2}}
 {id=WavePlayer_33, properties {soundWaveAssetPtr=/Game/Sounds/Fort_Impact_Sounds/Demolition/Fence/Fence_Impact_Med1}}
 {id=WavePlayer_34, properties {soundWaveAssetPtr=&Fence_Impact_Med2}}
 {id=WavePlayer_35, properties {soundWaveAssetPtr=/Game/Sounds/Fort_Impact_Sounds/Demolition/Fence/Fence_Impact_Med3}}
 {id=WavePlayer_5, properties {soundWaveAssetPtr=&Fence_Impact_Med1}}
 {id=WavePlayer_6, properties {soundWaveAssetPtr=&barbedwire_impact_01}}
 {id=WavePlayer_7, properties {soundWaveAssetPtr=&barbedwire_impact_02}}
 {id=WavePlayer_9, properties {soundWaveAssetPtr=&BarbedWire_Impact_opt}}
]
properties {
 attenuationSettings=/Game/Sounds/Attenuation/Fort_Demolition_Hit_Small_Sized_Object_Att
 maxDistance=4000
 soundClassObject=&Fort_Impacts
 volumeMultiplier=0.68
 concurrencyOverrides {maxCount=4}
 concurrencySet [&CONC_Object_Impacts_FourMax]
}
